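# Animation Blueprint for the BlockStack Aura character body, combining the attached parent's pose with one copied from the Head mesh component.
# A per-bone layered blend applies the head's pose on backpack and cape attach, pelvis and foot bone branches.
# Head comes from the customization-loaded event's part array.
# On FortPlayerMannequin owners, every update also scans skeletal mesh components for the Hope or ChunkBuggy BlockStack head meshes.
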AnimBlueprint /BRCosmetics/Characters/Player/Male/Medium/Bodies/M_MED_BlockStack_Aura/Meshes/M_MED_BlockStack_Aura_AnimBP
target_skeleton=&/M_MED_BlockStack_Aura_Skeleton
graphs [
 AnimGraph {nodes [
   n1 = Layered blend per bone(
    BasePose=Copy Pose From Mesh() {bUseAttachedParent=true}
    BlendPoses_0=Copy Pose From Mesh(SourceMeshComponent=Get Head() -> (Head:Skeletal Mesh Component Budgeted Object Reference)) {bUseAttachedParent=true}
   ) {layerSetup [{branchFilters [
       {boneName=attach_backpack}
       {boneName=attach_cape}
       {boneName=deform_foot_l}
       {boneName=deform_ball_l}
       {boneName=deform_foot_r}
       {boneName=deform_ball_r}
       {boneName=foot_bk_r}
       {boneName=foot_fr_r}
       {boneName=foot_bk_l}
       {boneName=foot_fr_l}
       {boneName=deform_pelvis}
      ]}]}
   Output Pose(Result=@n1) {layerGroup=None}
  ]}
 EventGraph {nodes [
   Event BlueprintUpdateAnimation()
   n1 = GetOwningActor()
   > Cast To FortPlayerMannequin!(execute, Object=@n1) -> (then, CastFailed, AsFort Player Mannequin:Fort Player Mannequin Object Reference) {is=title K2Node_DynamicCast}
   > n2 = For Each Loop(Array:Array of Skeletal Mesh Component Object References=Get Components By Class!(self:Actor Object Reference=@n1, ComponentClass:Actor Component Class Reference=/Script/Engine.SkeletalMeshComponent) -> (ReturnValue:Array of Skeletal Mesh Component Object References)) -> (Array Element:Skeletal Mesh Component Object Reference)
   n3 = Get SkeletalMesh!(self:Skinned Mesh Component Object Reference=@n2) -> (SkeletalMesh:Skeletal Mesh Object Reference)
   > Branch(Condition=OR Boolean(
     A="Equal (Object)"(A=@n3, B=/BRCosmetics/Characters/Player/Male/Medium/Bodies/M_MED_BlockStack_Hope/Meshes/M_MED_BlockStack_Hope_Head)
     B="Equal (Object)"(A=@n3, B=/BRCosmetics/Characters/Player/Male/Medium/Bodies/M_MED_BlockStack_ChunkBuggy/Meshes/M_MED_Blockstack_ChunkBuggy_Head)
    ))
   > n4 = Cast To SkeletalMeshComponentBudgeted!(execute, Object=@n2) -> (then, CastFailed, AsSkeletal Mesh Component Budgeted:Skeletal Mesh Component Budgeted Object Reference) {is=title K2Node_DynamicCast}
   > Set Head(Head:Skeletal Mesh Component Budgeted Object Reference=@n4) -> (Output_Get:Skeletal Mesh Component Budgeted Object Reference)
   n5 = Event OnCharacterCustomizationLoaded()
   > Set Head(Head=Get Array Item with EFort Custom Part Type(Array=@n5.CustomMeshParts, PartType=Head))
  ]}
]
variables [Head {type=SkeletalMeshComponentBudgeted}]
